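# Translucent, unlit, two-sided material for the Creative Audio Player device, apparently visualizing its attenuation volume.
# Object scale sets grid tiling, so cell count follows volume size, with sine-pulsed random blocks filling the cells.
# Opacity sums Fresnel-weighted grid and blocks, a depth-based intersection glow and a quantized sphere-mask ring, then applies vertical falloff and near-camera fade.
# A view-to-normal gradient shades the desaturated color and also scales opacity.
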Material /Game/Creative/Devices/Common/Audio/Materials/M_AudioPlayer_Attenuation
graph {
 properties {
  bCastRayTracedShadows=false
  bUseTranslucencyVertexFog=false
  blendMode=BLEND_Translucent
  pixelDepthOffsetMode=PDOM_Legacy
  shadingModel=MSM_Unlit
  twoSided=true
  usedShadingModels=Unlit
 }
 nodes [
  n1 = CameraVectorWS()
  n2 = PixelNormalWS()
  n3 = LinearInterpolate(Alpha=SmoothStep(
    Min=ScalarParameter() {defaultValue=0.5, parameterName=GradientMin, sliderMax=1}
    Max=ScalarParameter() {defaultValue=1, parameterName=GradientMax, sliderMax=1}
    Value=DotProduct(A=@n1, B=@n2)
   )) {constA=0.2, constB=0.8}
  Multiply(
   A=Desaturation(
    None=VectorParameter() {parameterName=Color, defaultValue=<rgba 1 1 1 1>}.RGB
    Fraction=ScalarParameter() {defaultValue=0.4, parameterName=DesaturationPercentage}
   )
   B=@n3
  )
  n4 = MaterialFunctionCall!(
   Distance Bias=ScalarParameter() {defaultValue=1, parameterName=DistanceBias}
   Distance Divisor=ScalarParameter() {defaultValue=3000, parameterName=Distance Divisor}
   Scene Depth
   Pixel Depth
  ) {materialFunction=/Game/Effects/Fort_Effects/Materials/_Functions/MF_DistanceBehindTranslucency_01}
  n5 = ScalarParameter() {defaultValue=20, parameterName=IntersectionExponent}
  n6 = OneMinus(None=MaterialFunctionCall(
    Brightness=ScalarParameter() {defaultValue=1, parameterName=FresnelBrightness}
    Exponent=ScalarParameter() {defaultValue=1, parameterName=FresnelExponent}
   ) {materialFunction=&MF_Fresnel_FX}.Fresnel Double Sided)
  n7 = LinearInterpolate(A=ScalarParameter() {defaultValue=1, parameterName=FresnelInnerOpacity}, Alpha=@n6)
  n8 = MaterialFunctionCall() {materialFunction=/Engine/Functions/Engine_MaterialFunctions02/WorldPositionOffset/ObjectScale}
  n9 = Divide(A=Divide(A=Add(A=@n8.Scale X, B=@n8.Scale X)), B=ScalarParameter() {defaultValue=50, desc=.75, parameterName=GridTileU})
  n10 = Divide(A=@n8.Scale Z, B=ScalarParameter() {defaultValue=100, parameterName=GridTileV})
  n11 = VertexInterpolator(VS=Ceil(None=AppendVector(A=@n9, B=@n10)))
  n12 = TextureSample(UVs=Multiply(A=TextureCoordinate(), B=@n11.PS)) {texture=/Game/Weapons/Prototype/ObjectMover/Textures/T_Movetool_Transition}
  n13 = LinearInterpolate(
   A=ScalarParameter() {defaultValue=0.5, parameterName=GridOpacityInner}
   B=ScalarParameter() {defaultValue=0.8, parameterName=GridOpacityOuter}
   Alpha=@n6
  )
  n14 = OneMinus(None=@n12.R)
  n15 = TextureCoordinate()
  n16 = Multiply(A=@n11.PS) {constB=4}
  n17 = Add(
   A=MaterialFunctionCall!(
    PrimaryTextureCoordinates=Multiply(A=@n15, B=@n16)
    2ndOrderTextureCoordinates
    RandomSeed
    NoiseDensityBias
    Bypass Noise Influence Crosstalk
   ) {materialFunction=/Game/Effects/Fort_Effects/Materials/_Functions/MF_RandomUVSample}.NoiseWithRange
   B=Multiply(A=Time(), B=ScalarParameter() {defaultValue=0.2, parameterName=BlockSpeed})
  )
  n18 = TwoSidedSign()
  n19 = Multiply(
   A=ConstantBiasScale(None=Sine(None=@n17))
   B=LinearInterpolate(
    A=ScalarParameter() {defaultValue=0.01, parameterName=BlocksOpacityInnerFaces}
    B=ScalarParameter() {defaultValue=0.01, parameterName=BlocksOpacity}
    Alpha=Saturate(None=@n18)
   )
  )
  n20 = Add(A=Add(A=Multiply(A=@n12.R, B=@n13), B=Multiply(A=@n14, B=@n19)), B=ScalarParameter() {defaultValue=0.001, parameterName=BlocksBase})
  n21 = TextureCoordinate()
  n22 = AppendVector(A=Constant(), B=ScalarParameter() {parameterName=VBias})
  n23 = ScalarParameter() {parameterName=RingSteps}
  n24 = Multiply(A=Multiply(A=@n11.PS, B=Add(A=@n21, B=@n22)), B=@n23)
  n25 = SphereMask(
   A=Divide(A=Floor(None=@n24), B=@n23)
   B=Divide(A=@n11.PS)
   Radius=ScalarParameter() {defaultValue=0.3, parameterName=ImpactParam}
   Hardness=Constant()
  )
  n26 = SphereMask(
   A=@n25
   B=ScalarParameter() {parameterName=RingCenter}
   Radius=ScalarParameter() {parameterName=RingRadius}
   Hardness=ScalarParameter() {parameterName=RingHardness}
  )
  n27 = Add(
   A=Add(A=Power(Base=@n4.Inverted, Exp=@n5), B=Multiply(A=@n7, B=@n20))
   B=Multiply(A=Multiply(A=@n25, B=@n26), B=ScalarParameter() {defaultValue=0.5, parameterName=RingOpacity})
  )
  n28 = SphereMask(
   A=ComponentMask(None=TextureCoordinate()) {g=true}
   B=ScalarParameter() {defaultValue=0.5, parameterName=FalloffGradientCenter}
   Radius=ScalarParameter() {defaultValue=1, parameterName=FalloffGradientRadius}
   Hardness=ScalarParameter() {parameterName=FalloffGradientHardness}
  )
  n29 = ScalarParameter() {defaultValue=1000, parameterName=NearCameraFadeDistance}
  Multiply(
   A=@n3
   B=Multiply(
    A=Multiply(A=@n27, B=@n28)
    B=MaterialFunctionCall(Distance=@n29) {materialFunction=&Near_Camera_Fade}.Result
   )
  )
 ]
}
