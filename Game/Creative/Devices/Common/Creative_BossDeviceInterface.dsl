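; Blueprint interface for Creative boss devices, declaring channel-trigger functions for the key events of a boss encounter.
; The events cover boss spawn and death, three phases, DBNO begin and end, player failure, and the reveal and destruction of horns and weakpoints.
; Each function is a bare declaration with no parameters or logic shown, leaving the behavior to implementers.
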(fn Interface_ChannelTrigger_BossDeath)

(fn Interface_ChannelTrigger_BossSpawned)

(fn Interface_ChannelTrigger_DBNOBegin)

(fn Interface_ChannelTrigger_DBNOEnd)

(fn Interface_ChannelTrigger_HornDestroyed)

(fn Interface_ChannelTrigger_HornRevealed)

(fn Interface_ChannelTrigger_Phase1)

(fn Interface_ChannelTrigger_Phase2)

(fn Interface_ChannelTrigger_Phase3)

(fn Interface_ChannelTrigger_PlayerFailure)

(fn Interface_ChannelTrigger_WeakpointDestroyed)

(fn Interface_ChannelTrigger_WeakpointRevealed)
