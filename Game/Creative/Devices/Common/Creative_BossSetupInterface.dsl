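; Creative_BossSetupInterface is a Blueprint interface for setting up bosses in Creative devices.
; It declares one function, Interface_BossSetup, which passes a boss placer actor reference and a Fort Mission Config Data class reference.
; The file holds only the signature, so the setup behavior is left to the classes that implement it.
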(fn Interface_BossSetup
  (params (BossPlacerActor :type Actor%20Object%20Reference) (BossMissionConfigData :type Fort%20Mission%20Config%20Data%20Class%20Reference)))
