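; Blueprint interface for channel weapons in Creative devices, declaring how an actor receives weapon channel and cooldown identifiers.
; SetWeaponChannelIDForActor and SetCooldownIDForActor each take a Gameplay Tag as the ID.
; Both are declarations with no logic of their own, leaving any behavior to implementing classes.
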(fn SetCooldownIDForActor
  (params (InCooldownID :type Gameplay%20Tag%20Structure))
  (return false))

(fn SetWeaponChannelIDForActor
  (params (InChannelID :type Gameplay%20Tag%20Structure))
  (return false))
