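; Blueprint function library of shared helpers for Fortnite Creative devices.
; Covers trigger eligibility with vehicle-driver resolution, game-phase and play-mode checks, team-to-index conversion and team color lookup, score awarding, ammo granting, gameplay effects and small math and curve utilities.
; Material helpers reuse or create dynamic material instances under deterministic GUID-based names, which team color updates use to set a TeamColor parameter.
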(fn ActorCanTrigger
  (params (Actor%20To%20Check :type Actor%20Object%20Reference) (Overlap%20Component :type Primitive%20Component%20Object%20Reference))
  (bind _actor_to_check (Variables|GetActor%20To%20Check))
  (bind _returnvalue (Ability|GetAbilitySystemComponent _actor_to_check))
  (bind _overlap_component (Variables|GetOverlap%20Component)
    (:then
      (if (Utilities|Enum|Equal(Enum) (Collision|GetCollisionObjectType _overlap_component) "ECC_Vehicle")
        (bind _returnvalue_2 (Vehicle|GetDriver(Message) (Variables|GetActor%20To%20Check)))
        (Utilities|IsValid _returnvalue_2
          (:Is%20Valid
            (LocalVariables|Setlocal%20controller (Pawn|GetController _returnvalue_2)))
          (:Is%20Not%20Valid
            _))
        (LocalVariables|Setlocal%20CanTrigger true)
        (else
          (Utilities|IsValid _returnvalue
            (:Is%20Valid
              (LocalVariables|Setlocal%20controller (Fortnite|GetFortPlayerControllerfromActor _actor_to_check))
              (LocalVariables|Setlocal%20CanTrigger (not (GameplayTags|TagContainer|CompareTagAssetInterfacetoOtherTagContainers (LocalVariables|GetDeviceDontTriggerTags) _returnvalue))))
            (:Is%20Not%20Valid
              (LocalVariables|Setlocal%20CanTrigger false)))))
      (return (LocalVariables|Getlocal%20CanTrigger) (LocalVariables|Getlocal%20controller)))
    (:else
      (return false 0))))

(fn AnimateFloatCurve
  (params (Delta%20Time :type Float%20%28double-precision%29) (Current%20Time :type Float%20%28double-precision%29%20%28by%20ref%29) (fCurve :type Curve%20Float%20Object%20Reference) (Animation%20Time :type Float%20%28double-precision%29 :default 1.0) (Invert%20Time :type Boolean))
  (bind _timealpha (LocalVariables|GetTimeAlpha))
  ;[1 1 1] Increment or Decrement value
  (Variables|SetBy-RefVar Current%20Time (Math|Float|Clamp(Float) (select (Variables|GetInvert%20Time) (- Current%20Time (Variables|GetDelta%20Time)) (+ Current%20Time (Variables|GetDelta%20Time))) 0.0 (Variables|GetAnimation%20Time)))
  ;[1 1 1] Get alpha value
  (LocalVariables|SetTimeAlpha (/ (Variables|GetCurrent%20Time) (Variables|GetAnimation%20Time)))
  ;[1 1 1] Get curve float value, returning the straight alpha value (i.e. lerp) if we don't input a curve.
  (bind _fcurve (Variables|GetfCurve)
    (:then
      (LocalVariables|SetCurveFloatResult (Math|Curves|GetFloatValue _fcurve _timealpha)))
    (:else
      (LocalVariables|SetCurveFloatResult _timealpha)))
  (return (LocalVariables|GetCurveFloatResult) (== (Variables|GetCurrent%20Time) (select (Variables|GetInvert%20Time) 0.0 (Variables|GetAnimation%20Time)))))

(fn ApplyVentRecentlyLaunchedTagorAbortLaunch
  (params (Actor%20To%20Launch :type Actor%20Object%20Reference) (Vent%20Actor :type Actor%20Object%20Reference))
  (Utilities|IsValid Actor%20To%20Launch
    (:Is%20Valid
      (Utilities|IsValid Vent%20Actor
        (:Is%20Valid
          (bind _returnvalue (GameplayTags|HasMatchingGameplayTag(Message) Actor%20To%20Launch "(TagName=\"Abilities.Generic.Vent.RecentlyLaunched\")"))
          (if (not _returnvalue)
            (Fortnite|ApplyGameplayEffectsToActor Actor%20To%20Launch (Utilities|Array|MakeArray "/Game/Creative/BuildingActors/GameplayActors/GE_LaunchedByVent.GE_LaunchedByVent_C") (Transformation|GetActorLocation Vent%20Actor) false 0 (Ability|GetAbilitySystemComponent Vent%20Actor))
            ;[1 0.5104 0.5104] Return should launch
            (return true)
            (else
              _)))
        (:Is%20Not%20Valid
          _))
      ;[1 0.5104 0.5104] Either we recently launched OR vital data is missing.
      ; Return should NOT launch.
      (return false))
    (:Is%20Not%20Valid
      _)))

(fn AwardScore
  (params (Actor :type Actor%20Object%20Reference) (Score :type Integer) (Type :type EAwardScoreType%20Enum :default "NewEnumerator0") (Controller :type Controller%20Object%20Reference) (Minigame%20Logic :type Fort%20Minigame%20Logic%20Component%20Object%20Reference) (TargetTags :type Gameplay%20Tag%20Container%20Structure :default "(GameplayTags=((TagName=\"Minigame.Stat.Score\")))"))
  (bind (_outsourcetags _outcontexttags) (Class|FortMinigameStatsComponent|GetSourceandContextTags (LocalVariables|Getlocal%20MinigameStatsComp) (LocalVariables|GetParticipant%20Controller)))
  (if (Utilities|Enum|NotEqual(Enum) Type "NewEnumerator0")
    (LocalVariables|SetParticipant%20Controller (CreativeAI|Controller|GetParticipantControllerfromActor (Variables|GetController) false (Variables|GetActor)))
    (bind _participant_controller (LocalVariables|GetParticipant%20Controller)
      (:then
        (Utilities|IsValid (Minigame|Stats|GetMinigameStatsComponent _participant_controller)
          (:Is%20Valid
            (bind _returnvalue (CreativeDevices|GetMinigameStatsComp (LocalVariables|GetParticipant%20Controller) "Athena"))
            (LocalVariables|Setlocal%20MinigameStatsComp _returnvalue)
            (LocalVariables|Getlocal%20MinigameStatsComp
              (:then
                (LocalVariables|Setlocal%20Score (Utilities|Select :Index (Variables|GetType) :NewEnumerator0 0 :NewEnumerator1 (Variables|GetScore) :NewEnumerator2 (* (Variables|GetScore) -1) :NewEnumerator3 (- (Variables|GetScore) (Minigame|Stats|GetStatforPlayer (Minigame|Stats|GetMinigameStatsComponent _participant_controller) (Utilities|Casting|CastToFortPlayerState (Variables|/Script/Engine.Controller|GetPlayerState _participant_controller)) (Class|FortMinigameStatsComponent|GetLoadedStatfromClass (Minigame|Stats|GetMinigameStatsComponent _participant_controller) "/Game/Athena/Playset/Minigames/StatFilters/MinigameStat_Score.MinigameStat_Score_C") false))))
                (Class|FortMinigameStatsComponent|SendStatEventwithTags (LocalVariables|Getlocal%20MinigameStatsComp) (LocalVariables|GetParticipant%20Controller) "MinigameDynamicEvent" (Variables|GetActor) TargetTags _outsourcetags _outcontexttags (LocalVariables|Getlocal%20Score)))
              (:else)))))
      (:else))))

(fn BooltoBoolwithUnset
  (params (Input :type Boolean))
  ;[0 0.8698 0.7627 0.25] Very simple helper function to convert a bool to an enum that includes an Unset value (for cases where we want to know verifiably whether a boolean has been set or is uninitialized.
  (return (select Input "NewEnumerator2" "NewEnumerator1")))

(fn CurrentMaterialMatchesBase
  (params (Static%20Mesh%20Component :type Static%20Mesh%20Component%20Object%20Reference) (MatID :type Integer))
  (bind _returnvalue (Rendering|Material|GetMaterial Static%20Mesh%20Component 0))
  (bind _staticmesh (Variables|/Script/Engine.StaticMeshComponent|GetStaticMesh Static%20Mesh%20Component)
    (:then
      (if (Utilities|Array|IsValidIndex (Variables|/Script/Engine.StaticMesh|GetStaticMaterials _staticmesh) MatID)
        (Utilities|IsValid _returnvalue
          (:Is%20Valid
            (return (== _returnvalue (Utilities|Struct|BreakStaticMaterial (Utilities|Array|Get(acopy) (Variables|/Script/Engine.StaticMesh|GetStaticMaterials _staticmesh)))) (Utilities|Struct|BreakStaticMaterial (Utilities|Array|Get(acopy) (Variables|/Script/Engine.StaticMesh|GetStaticMaterials _staticmesh))))))))
    (:else)))

(fn GenerateUniqueMaterialName
  (params (PrimitiveComponent :type Primitive%20Component%20Object%20Reference) (Name :type String))
  (bind _returnvalue (Components|GetOwner PrimitiveComponent))
  ;[0.8827 0.1953 1 0.25] Creating a unique but deterministic Optional Name (needed in case this is used at construction
  (Utilities|IsValid PrimitiveComponent
    (:Is%20Valid
      (bind _returnvalue_2 (GUID|GenerateGuidfromObject PrimitiveComponent true))
      (Utilities|IsValid _returnvalue
        (:Is%20Valid
          (bind _returnvalue_3 (GUID|GenerateGuidfromObject _returnvalue true))
          (return (Utilities|String|StringToName (Utilities|String|Left (Utilities|String|Append Name "-" (Utilities|String|Right (Guid|ToString(Guid) _returnvalue_3) 8) (Utilities|String|Right (Guid|ToString(Guid) _returnvalue_2) 8)) (LocalVariables|GetMax%20Length)))))))))

(fn GetDriverorPawnController
  (params (Actor%20To%20Check :type Actor%20Object%20Reference))
  (bind _actor_to_check (Variables|GetActor%20To%20Check))
  (bind _ascontroller (Utilities|Casting|CastToController Actor%20To%20Check)
    (:then
      (return true _ascontroller))
    (:CastFailed
      (bind _returnvalue (Vehicle|GetDriver(Message) (Variables|GetActor%20To%20Check)))
      (Utilities|IsValid _returnvalue
        (:Is%20Valid
          (LocalVariables|Set_Controller (Pawn|GetController _returnvalue)))
        (:Is%20Not%20Valid
          (Utilities|IsValid (Ability|GetAbilitySystemComponent _actor_to_check)
            (:Is%20Valid
              (LocalVariables|Set_Controller (Fortnite|GetFortPlayerControllerfromActor _actor_to_check)))
            (:Is%20Not%20Valid
              _))))
      (return (Utilities|IsValid (LocalVariables|Get_Controller) :!class /Script/Engine.KismetSystemLibrary) (LocalVariables|Get_Controller)))) ; we're already a controller,\njust return it.
)

(fn GetDynamicTeamColor
  (params (Actor :type Actor%20Object%20Reference) (Minigame :type Fort%20Minigame%20Object%20Reference) (Color%20Type :type ECreativeColorSetType%20Enum :default "NewEnumerator0"))
  (bind _minigame (Variables|GetMinigame))
  (bind _returnvalue (Minigame|GetMinigameTeamComponent _minigame))
  (switch Utilities|FlowControl|Switch|SwitchonECreativeColorSetType Color%20Type
    (:NewEnumerator0
      _)
    (:NewEnumerator1
      _)
    (:NewEnumerator2
      ;[1 1 1] Use Team Color for the incoming pawn
      (Utilities|IsValid _returnvalue
        (:Is%20Valid
          (return (Minigame|GameRules|GetTeamColorIndex _returnvalue (Math|Byte|Byte+Byte (Utilities|Enum|LiteralenumEFortTeam "HumanPvP_Team1") (Team|GetActorPvPTeamIndex (Variables|GetActor))))))
        (:Is%20Not%20Valid
          (return -1))))
    (:NewEnumerator3
      ;[1 1 1] Use Team Relationship between incoming pawn and local pawn
      (return (Math|Conversions|ToInteger(Byte) (Utilities|Select :Index (Team|GetTeamAffiliation (Variables|GetActor) (Fortnite|GetFirstLocalFortPlayerController true)) :Friendly "NewEnumerator1" :Neutral "NewEnumerator4" :Hostile "NewEnumerator2")))))
  (return -1))

(fn GetPlayerStatefromNetID
  (params (Unique%20Net%20Id :type Unique%20Net%20Id%20Structure%20%28by%20ref%29))
  (bind _returnvalue (Game|GetGameState 0))
  (Utilities|IsValid _returnvalue
    (:Is%20Valid
      (for _array_element (Variables|/Script/Engine.GameStateBase|GetPlayerArray _returnvalue)
        (bind _asfort_player_state (Utilities|Casting|CastToFortPlayerState _array_element))
        (if (Online|UniqueNetIdRepl|Equal(UniqueIdRepl) (PlayerState|GetPlayerUniqueNetID _asfort_player_state) Unique%20Net%20Id)
          (return _asfort_player_state)))
      (return (LocalVariables|GetPurposelyInvalidPlayerState)))))

(fn GetTeamColorForTeamOption
  (params (ContextActor :type Actor%20Object%20Reference) (TeamOption :type Creative%20Team%20Option%20Structure))
  (bind (_teamtype _teamindex) (Utilities|Struct|BreakCreativeTeamOption TeamOption))
  (bind _returnvalue (Minigame|GetMinigame ContextActor :!class /Script/FortniteGame.MinigameBlueprintLibrary))
  (bind _returnvalue_2 (Minigame|GetMinigameTeamComponent _returnvalue))
  (bind (_teamlightcolors _teamselectorcolors _teamboldcolors) (Class|GetClassDefaults "/Game/Creative/Sets/MilitaryBase/Props/CreativeTeamColors.CreativeTeamColors_C"))
  (Utilities|IsValid ContextActor
    (:Is%20Valid
      (Utilities|IsValid _returnvalue
        (:Is%20Valid
          (Utilities|IsValid _returnvalue_2
            (:Is%20Valid
              (LocalVariables|SetColorIndex (Minigame|GameRules|GetTeamColorIndex _returnvalue_2 (Creative|TeamOption|ConverttoTeam TeamOption))))
            (:Is%20Not%20Valid
              (LocalVariables|SetColorIndex (Math|Integer|SelectInt _teamindex 0 (Utilities|Enum|Equal(Enum) _teamtype "TeamIndex")))))
          (return (Utilities|Array|Get(acopy) _teamselectorcolors (Math|Integer|Max(Integer) (Math|Integer|%25%28Integer%29 (LocalVariables|GetColorIndex) (Utilities|Array|Length _teamlightcolors)))) (Utilities|Array|Get(acopy) _teamlightcolors (Math|Integer|Max(Integer) (Math|Integer|%25%28Integer%29 (LocalVariables|GetColorIndex) (Utilities|Array|Length _teamlightcolors)))) (Utilities|Array|Get(acopy) _teamboldcolors (Math|Integer|Max(Integer) (Math|Integer|%25%28Integer%29 (LocalVariables|GetColorIndex) (Utilities|Array|Length _teamlightcolors))))))
        (:Is%20Not%20Valid
          _)))
    (:Is%20Not%20Valid
      _)))

(fn GetTeamColorIndexfromTag
  (params (Actor%20with%20Tag :type Actor%20Object%20Reference) (Minigame :type Fort%20Minigame%20Object%20Reference))
  (bind _returnvalue (GameplayTags|GetOwnedGameplayTags (Variables|GetActor%20with%20Tag) :!class /Script/GameplayTags.GameplayTagAssetInterface))
  (bind _output_get (LocalVariables|SetGameplayTags (GameplayTags|BreakGameplayTagContainer _returnvalue)))
  ;[1 1 1] Work backwards through the gameplay tags, we want to find the Last color tag added.
  ; (In the case of multiple effects being applied)
  (bind (_arrayindex _arrayelement) (Utilities|Array|ReverseforEachLoop _output_get)
    (:LoopBody
      (bind _output_get_2 (LocalVariables|SetCheck%20Tag _arrayelement))
      (if (GameplayTags|MatchesTag _output_get_2 "(TagName=\"Creative.Toys.Abilities.PlayerVFX.Color\")")
        (if (GameplayTags|Equal(GameplayTag) (LocalVariables|GetCheck%20Tag) "(TagName=\"Creative.Toys.Abilities.PlayerVFX.Color.Team\")")
          ;;[1 1 1] Team Color Check
          (return (Teams|GetDynamicTeamColor (Variables|GetActor%20with%20Tag) (Variables|GetMinigame) "NewEnumerator2" (Variables|GetMinigame)) "NewEnumerator2")
          (elif (GameplayTags|Equal(GameplayTag) (LocalVariables|GetCheck%20Tag) "(TagName=\"Creative.Toys.Abilities.PlayerVFX.Color.TeamRelationship\")")
            ;;[1 1 1] Team Relationship Check
            (return (Teams|GetDynamicTeamColor (Variables|GetActor%20with%20Tag) (Variables|GetMinigame) "NewEnumerator3" (Variables|GetMinigame)) "NewEnumerator3")))))))

(fn GetTeamColorIndexorTeamIndex
  (params (Minigame%20Logic :type Fort%20Minigame%20Logic%20Component%20Object%20Reference) (Team%20Index :type Integer))
  (Utilities|IsValid Minigame%20Logic
    (:Is%20Valid
      (bind _currentminigame (Variables|/Script/FortniteGame.FortMinigameLogicComponent|GetCurrentMinigame Minigame%20Logic)
        (:then
          (Utilities|IsValid (Minigame|GetMinigameTeamComponent _currentminigame)
            (:Is%20Valid
              (return (select (== (Minigame|GameRules|GetTeamColorIndex (Minigame|GetMinigameTeamComponent _currentminigame) (Teams|IndextoTeam Team%20Index Minigame%20Logic)) -1) Team%20Index (Minigame|GameRules|GetTeamColorIndex (Minigame|GetMinigameTeamComponent _currentminigame) (Teams|IndextoTeam Team%20Index Minigame%20Logic)))))
            (:Is%20Not%20Valid
              _)))
        (:else
          _))
      (return Team%20Index))
    (:Is%20Not%20Valid
      _)))

(fn GetMID
  (params (Target :type Primitive%20Component%20Object%20Reference))
  (bind _returnvalue (Rendering|Material|GetMaterial Target 0))
  (bind _target (Variables|GetTarget))
  (Utilities|IsValid Target
    (:Is%20Valid
      (bind _asmaterial_instance_dynamic (Utilities|Casting|CastToMaterialInstanceDynamic _returnvalue)
        (:then
          (return _asmaterial_instance_dynamic))
        (:CastFailed
          (Utilities|IsValid _returnvalue
            (:Is%20Valid
              (bind _uniquename (Material|GenerateUniqueMaterialName _target (Utilities|String|Append (Utilities|GetDisplayName _returnvalue) "_" "0") _target))
              (bind _returnvalue_2 (Rendering|Material|CreateDynamicMaterialInstance _target 0 0 _uniquename :!class /Script/Engine.PrimitiveComponent))
              (return _returnvalue_2))
            (:Is%20Not%20Valid
              (return 0))))))
    (:Is%20Not%20Valid
      (return 0))))

(fn GetMID_FromMaterial
  (params (Target :type Primitive%20Component%20Object%20Reference) (Index :type Integer) (SourceMaterial :type Material%20Interface%20Object%20Reference))
  (bind _sourcematerial (Variables|GetSourceMaterial))
  (bind _asmaterial_instance_dynamic (Utilities|Casting|CastToMaterialInstanceDynamic (Rendering|Material|GetMaterial Target Index))
    (:then
      (bind _parent (Variables|/Script/Engine.MaterialInstance|GetParent _asmaterial_instance_dynamic)
        (:then
          (if (== _parent _sourcematerial)
            (return _asmaterial_instance_dynamic)
            (else
              _)))
        (:else)))
    (:CastFailed
      _))
  (bind _uniquename (Material|GenerateUniqueMaterialName (Variables|GetTarget) (Utilities|String|Append (Utilities|GetDisplayName _sourcematerial) "_" (Utilities|String|ToString(Integer) (Variables|GetIndex))) (Variables|GetTarget)))
  (bind _returnvalue (Rendering|Material|CreateDynamicMaterialInstance (Variables|GetTarget) (Variables|GetIndex) SourceMaterial _uniquename :!class /Script/Engine.PrimitiveComponent))
  (return _returnvalue))

(fn GiveAmmoToInventoryOwner
  (params (Inventory%20Owner :type Fort%20Inventory%20Owner%20Interface%20Interface) (World%20Item%20Definition :type Fort%20World%20Item%20Definition%20Object%20Reference) (Pickup%20Instigator%20Handle :type Integer :default -1) (AmmoToGiveItemDefinitionMap :type Map%20of%20Fort%20World%20Item%20Definition%20Soft%20Object%20References%20to%20Integers) (DefaultAmmoToGive :type Integer) (SpareWeaponAmmo_Override :type Boolean) (SpareWeaponAmmo :type Integer))
  (bind (_value _returnvalue) (Utilities|Map|Find (Variables|GetAmmoToGiveItemDefinitionMap) (Utilities|ResolveSoftReference (LocalVariables|GetTargetAmmoDef))))
  (bind _world_item_definition (Variables|GetWorld%20Item%20Definition))
  (bind _targetammodef (LocalVariables|GetTargetAmmoDef))
  (Utilities|FlowControl|Sequence
    (:then_0
      ;[1 0.4843 0.1465] Inspect the Item Definition for reasons to return early without granting ammo
      (Utilities|IsValid World%20Item%20Definition
        (:Is%20Valid
          (bind _asfort_weapon_item_definition (Utilities|Casting|CastToFortWeaponItemDefinition World%20Item%20Definition))
          (if (and (Utilities|IsValidClass (WeaponItemDefinition|GetReloadAbility _asfort_weapon_item_definition)) (or (not (Variables|GetSpareWeaponAmmo_Override)) (> (Variables|GetSpareWeaponAmmo) 0)))
            _
            (else
              _)))
        (:Is%20Not%20Valid
          _))
      ;;[1 0.4635 0.4635] Early Escape Conditions
      (return))
    (:then_1
      ;[1 0.4462 0] Validate remaining data (maybe return early) and create local variables
      (bind _ascontroller (Utilities|Casting|CastToController Inventory%20Owner)
        (:then
          (LocalVariables|SetController _ascontroller)
          (bind _output_get (LocalVariables|SetTargetAmmoDef (WeaponItemDefinition|GetAmmoWorldItemDefinition _world_item_definition)))
          (if (and (Utilities|IsValid _output_get :!class /Script/Engine.KismetSystemLibrary) (!= _output_get _world_item_definition))
            (if (or (== (Inventory|GetItemQuantityOnPlayer (LocalVariables|GetController) _output_get) 0) (Variables|GetSpareWeaponAmmo_Override))
              _
              (else
                _))
            (else
              _))
          ;;[1 0.4635 0.4635] Early Escape Conditions
          (return))
        (:CastFailed
          _)))
    (:then_2
      ;[0 0.1005 1] Find the amount of ammo to give and give it.
      (bind _output_get_2 (LocalVariables|SetNumAmmoToGive (select (Variables|GetSpareWeaponAmmo_Override) (Variables|GetSpareWeaponAmmo) (select _returnvalue _value (Variables|GetDefaultAmmoToGive)))))
      (if (> _output_get_2 0)
        (Inventory|GiveItemtoInventoryOwner :InventoryOwner (Variables|GetInventory%20Owner) :ItemDefinition _targetammodef :NumberToGive _output_get_2 :bNotifyPlayer true :ItemLevel -1 :PickupInstigatorHandle (Variables|GetPickup%20Instigator%20Handle) :bUseItemPickupAnalyticEvent false :WeaponAmmoOverride -1)))))

(fn IndexToTeam
  (params (Index :type Integer))
  ;[1 1 1] The Creative Team Byte starts at Index 3 for Team 1, rather than the expected Index 0. As such, this converts an index to a team to help encapsulate this decision and reduce confusion for designers
  (return (Math|Byte|Byte+Byte (Utilities|Enum|LiteralenumEFortTeam "HumanPvP_Team1") (Math|Conversions|ToByte(Integer) Index))))

(fn IsMatchingChannelID
  (params (A :type Fort%20Gameplay%20Message%20Component%20Base%20Object%20Reference) (B :type Fort%20Gameplay%20Message%20Component%20Base%20Object%20Reference))
  ;[0.2703 0 1 0.5] Does the creative channel of component A match that of B? Useful if we want to ensure that if different channels are called at the same time we call them in a specific order.
  (return (GameplayTags|Equal(GameplayTagContainer) (Class|FortGameplayMessageComponentBase|GetChannelId A) (Class|FortGameplayMessageComponentBase|GetChannelId B))))

(fn IsActivatedOnPhase
  (params (State :type EFortMinigameState%20Enum :default "PreGame") (PhaseIndex :type Integer) (Actor :type Actor%20Object%20Reference))
  ;[0 0.3855 1] Phase Matching for the 'Activate on Game Phase' common user option.
  (return (Math|Boolean|ORBoolean (Math|Boolean|ANDBoolean (== (Variables|GetPhaseIndex) 1) (CallFunction|IsinPlayMode (Variables|GetActor) (Variables|GetActor)) (Utilities|Enum|Equal(Enum) (Variables|GetState) (Utilities|Enum|LiteralenumEFortMinigameState "PreGame"))) (and (== (Variables|GetPhaseIndex) 2) (Utilities|Enum|Equal(Enum) (Variables|GetState) "Warmup")) (and (== (Variables|GetPhaseIndex) 3) (Utilities|Enum|Equal(Enum) (Variables|GetState) "InProgress")))))

(fn IsEnabledPhaseMatch
  (params (State :type EFortMinigameState%20Enum :default "PreGame") (PhaseIndex :type Integer) (Actor :type Actor%20Object%20Reference))
  ;[0 0.3855 1] Phase Matching for the 'Enabled During Phase' nodes. Phase Index matches 'Knob_Device_Common_EnabledOnPhase' and 'Knob_Device_Common_EnabledOnPhase_DefaultAlways'
  (return (Math|Boolean|ORBoolean (== (Variables|GetPhaseIndex) 0) (and (== (Variables|GetPhaseIndex) 2) (Utilities|Enum|Equal(Enum) (Variables|GetState) "InProgress")) (Math|Boolean|ANDBoolean (== (Variables|GetPhaseIndex) 1) (Math|Byte|Byte<Byte (Variables|GetState) (Utilities|Enum|LiteralenumEFortMinigameState "InProgress")) (CallFunction|IsinPlayMode (Variables|GetActor) (Variables|GetActor))) (or (Math|Boolean|ANDBoolean (== (Variables|GetPhaseIndex) 3) (Utilities|Enum|Equal(Enum) (Variables|GetState) (Utilities|Enum|LiteralenumEFortMinigameState "PreGame")) (not (CallFunction|IsinPlayMode (Variables|GetActor) (Variables|GetActor)))) (and (== (Variables|GetPhaseIndex) 3) (Editor|InEditor))) (and (== (Variables|GetPhaseIndex) 3) (Utilities|Enum|Equal(Enum) (Variables|GetState) (Utilities|Enum|LiteralenumEFortMinigameState "PostGameReset"))))))

(fn IsInEditor
  ;[1 0 0.0426] Deprecated really, this function was here for before we had 'InEditor', now just links to it. Shouldn't use.
  (return (Editor|InEditor)))

(fn IsInPlayMode
  (params (Actor :type Actor%20Object%20Reference%20%28by%20ref%29))
  (return (Minigame|IsinPlayMode Actor :!class /Script/FortniteGame.MinigameBlueprintLibrary)))

(fn IsLocalPawn
  (params (Actor :type Actor%20Object%20Reference))
  (Networking|CanExecuteCosmeticEvents
    (:True
      (for _array_element (Fortnite|GetLocalFortPlayerControllers true)
        (Utilities|IsValid (Pawn|GetControlledPawn _array_element)
          (:Is%20Valid
            (if (Utilities|Equal(Object) (Pawn|GetControlledPawn _array_element) Actor)
              (return true)))) ; Check against valid as we don't want \nNull=Null to succeed.
      ))
    (:False
      _))
  (return false))

(fn IsLogicInPlayMode
  (params (Logic :type Fort%20Minigame%20Logic%20Component%20Object%20Reference%20%28by%20ref%29))
  (Utilities|IsValid Logic
    (:Is%20Valid
      (bind _currentminigame (Variables|/Script/FortniteGame.FortMinigameLogicComponent|GetCurrentMinigame Logic)
        (:then
          (return (Minigame|IsinPlayMode _currentminigame :!class /Script/FortniteGame.FortMinigame)))
        (:else
          _)))
    (:Is%20Not%20Valid
      _))
  (return false))

(fn IsOnUnpublishedIsland
  (params (Actor :type Actor%20Object%20Reference))
  (bind _returnvalue (Class|CreativeBlueprintLibrary|GetExperienceType Actor))
  (return (Utilities|Enum|Equal(Enum) _returnvalue "Island")))

(fn IsValidTeam
  (params (Team%20to%20Check :type Integer) (Actor%20to%20Check :type Actor%20Object%20Reference))
  (bind _team_to_check (Variables|GetTeam%20to%20Check))
  (bind _returnvalue (== _team_to_check -1))
  ;[0 0.3855 1] Simple team check for an actor versus an index. Index matches 'Knob_Common_TeamIndex'
  (Utilities|IsValid Actor%20to%20Check
    (:Is%20Valid
      (return (or (Math|Byte|Equal(Byte) (Team|GetActorPvPTeamIndex Actor%20to%20Check) (Math|Conversions|ToByte(Integer) _team_to_check)) _returnvalue)))
    (:Is%20Not%20Valid
      (return _returnvalue))))

(fn MatchesTagReturnMatch
  (params (TagContainer_GameplayTags :type Array%20of%20Gameplay%20Tag%20Structures) (Tag :type Gameplay%20Tag%20Structure) (ExactMatch :type Boolean))
  (bind _array_element (Utilities|Array|ForEachLoopwithBreak TagContainer_GameplayTags)
    (:LoopBody
      (if (GameplayTags|MatchesTag _array_element Tag ExactMatch)
        (LocalVariables|SetFoundTag _array_element)))
    (:Completed
      _))
  (return (GameplayTags|IsGameplayTagValid (LocalVariables|GetFoundTag)) (LocalVariables|GetFoundTag)))

(fn ReattachorCreateMID
  (params (Mesh :type Primitive%20Component%20Object%20Reference) (Target%20MID :type Material%20Instance%20Dynamic%20Object%20Reference%20%28by%20ref%29) (MatId :type Integer))
  (Utilities|IsValid Target%20MID
    (:Is%20Valid
      (Variables|GetMesh
        (:then
          (if (Utilities|NotEqual(Object) (Rendering|Material|GetMaterial (Variables|GetMesh) (Variables|GetMatId)) (Variables|GetTarget%20MID))
            (Rendering|Material|SetMaterial (Variables|GetMesh) (Variables|GetMatId) (Variables|GetTarget%20MID))))
        (:else)))
    (:Is%20Not%20Valid
      (bind _mesh (Variables|GetMesh)
        (:then
          (bind _uniquename (Material|GenerateUniqueMaterialName _mesh (Utilities|String|Append (Utilities|GetDisplayName (Rendering|Material|GetMaterial Mesh MatId)) "_" (Utilities|String|ToString(Integer) MatId)) _mesh))
          (bind _returnvalue (Rendering|Material|CreateDynamicMaterialInstance (Variables|GetMesh) (Variables|GetMatId) 0 _uniquename :!class /Script/Engine.PrimitiveComponent))
          (Variables|SetBy-RefVar (Variables|GetTarget%20MID) _returnvalue))
        (:else)))) ; If the target MID isn't valid, create a new one.
)

(fn RefreshTeamChangeBinding
  (params (Minigame%20Logic :type Fort%20Minigame%20Logic%20Component%20Object%20Reference) (Unbind :type Boolean) (Event :type Delegate%20%28by%20ref%29))
  ;[1 0.6061 0] This unbinds or binds an event to the Team Config Updated event. Used for Devices that have team colors so that they update when the team settings do.
  (Utilities|IsValid Minigame%20Logic
    (:Is%20Valid
      (Networking|CanExecuteCosmeticEvents
        (:True
          (bind _currentminigame (Variables|/Script/FortniteGame.FortMinigameLogicComponent|GetCurrentMinigame Minigame%20Logic)
            (:then
              (if Unbind
                (Minigame|UnbindEventfromMinigameTeamConfigUpdated _currentminigame Event)
                (else
                  (Minigame|BindEventtoMinigameTeamConfigUpdated _currentminigame Event))))
            (:else)))))))

(fn SimpleApplyGameplayEffecttoActor
  (params (GameplayEffect :type Gameplay%20Effect%20Class%20Reference) (Actor :type Actor%20Object%20Reference))
  (if (Utilities|IsValidClass GameplayEffect)
    (bind _actor (Variables|GetActor)
      (:then
        (Fortnite|ApplyGameplayEffectsToActor _actor (Utilities|Array|MakeArray (Variables|GetGameplayEffect)) (Transformation|GetActorLocation _actor)))
      (:else))))

(fn TeamToIndex
  (params (Team :type Byte))
  ;[1 1 1] The Creative Team Byte starts at Index 3 for Team 1, rather than the expected Index 0. As such, this converts a team to an index to help encapsulate this decision and reduce confusion for designers.
  (return (Math|Conversions|ToInteger(Byte) (Math|Byte|Byte-Byte Team (Utilities|Enum|LiteralenumEFortTeam "HumanPvP_Team1")))))

(fn UpdateTeamColor
  (params (Minigame%20Logic :type Fort%20Minigame%20Logic%20Component%20Object%20Reference) (Mesh :type Primitive%20Component%20Object%20Reference) (Team :type Integer))
  ;[1 0.6061 0] Sets the Team Color of a mesh (that has a 'TeamColor' material setting). Uses the TeamSetting common useroption
  (Rendering|Material|SetScalarParameterValue (Material|GetMID Mesh Mesh) "TeamColor" (Math|Conversions|ToFloat(Integer) (Teams|GetTeamColorIndexOrTeamIndex Minigame%20Logic Team Mesh)) :!class /Script/Engine.MaterialInstanceDynamic))

(fn WrappingModulo
  (params (A :type Integer) (B :type Integer))
  (bind _output_get (LocalVariables|Set%25%20Result (Math|Integer|%25%28Integer%29 A B)))
  (if (< _output_get 0)
    (return (+ B _output_get))
    (else
      (return (LocalVariables|Get%25%20Result)))))
