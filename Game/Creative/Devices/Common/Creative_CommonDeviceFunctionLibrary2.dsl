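; Blueprint function library of shared helpers for Creative devices, covering game-phase user options, colors and attenuation settings.
; Phase functions test minigame state against activate and enabled options, with "Game Start" left to PostMinigameStarted binding.
; Converters shift legacy option integers, which start at -1, to and from enum values.
; Lookups map range to attenuation, read data table colors by row index, and match team colors by near-equality to tolerate float error.
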(fn ActivateOnPhase
  (params (Minigame :type Fort%20Minigame%20Object%20Reference) (ActivatePhase :type EActivateOnPhase%20Enum :default "NewEnumerator0"))
  (bind _currentstate (Variables|/Script/FortniteGame.FortMinigame|GetCurrentState Minigame))
  ;[0 0.3855 1 0.2] Phase Matching for the 'Activate on Game Phase' common user option.
  ; This does not cover the 'Game Start' case. Bind to PostMinigameStarted for that check.
  (Utilities|IsValid Minigame
    (:Is%20Valid
      (switch Utilities|FlowControl|Switch|SwitchonEActivateOnPhase ActivatePhase
        (:NewEnumerator0
          _)
        (:NewEnumerator1
          (return (and (Minigame|IsinPlayMode Minigame :!class /Script/FortniteGame.FortMinigame) (Utilities|Enum|Equal(Enum) _currentstate (Utilities|Enum|LiteralenumEFortMinigameState "PreGame")))))
        (:NewEnumerator2
          (return (Utilities|Enum|Equal(Enum) _currentstate "Warmup")))
        (:NewEnumerator3
          (return false))))
    (:Is%20Not%20Valid
      _))
  (return false))

(fn EnabledPhaseEnumtoLegacyInt
  (params (EnabledDuringPhase :type Byte))
  ;[1 1 1] Enum values are shifted by 1 from UserOption values (which start at -1)
  (return (- (Math|Conversions|ToInteger(Byte) EnabledDuringPhase) 1)))

(fn EnabledPhaseMatch
  (params (State :type EFortMinigameState%20Enum :default "PreGame") (EnabledPhase :type EEnabledOnPhase%20Enum :default "NewEnumerator1") (Actor :type Actor%20Object%20Reference))
  ;[0 0.3855 1] Phase Matching for the 'Enabled During Phase' nodes. Phase Index matches 'Knob_Device_Common_EnabledOnPhase' and 'Knob_Device_Common_EnabledOnPhase_DefaultAlways'
  (switch Utilities|FlowControl|Switch|SwitchonEEnabledOnPhase EnabledPhase
    (:NewEnumerator1
      (return false))
    (:NewEnumerator0
      (return true))
    (:NewEnumerator2
      (return (and (Math|Byte|Byte<Byte (Variables|GetState) (Utilities|Enum|LiteralenumEFortMinigameState "InProgress")) (Class|CreativeCommonDeviceFunctionLibrary|IsinPlayMode (Variables|GetActor) (Variables|GetActor)))))
    (:NewEnumerator4
      (return (Utilities|Enum|Equal(Enum) (Variables|GetState) "InProgress")))
    (:NewEnumerator3
      (return (and (Utilities|Set|ContainsItem (LocalVariables|GetEditStates) (Variables|GetState)) (or (Editor|InEditor) (not (Class|CreativeCommonDeviceFunctionLibrary|IsinPlayMode (Variables|GetActor) (Variables|GetActor)))))))))

(fn GetCreativeTeamColorIndexfromColor
  (params (Color :type Linear%20Color%20Structure))
  ;[0.06848 0.6038 1 0.25] Looking up our color in the Team Light Colors to find out which Team Color it matches. Manually iterating through the list rather than using a Find in order to allow for floating point errors breaking the search
  (bind _output_get (LocalVariables|SetColor%20Array (Class|GetClassDefaults "/Game/Creative/Sets/MilitaryBase/Props/CreativeTeamColors.CreativeTeamColors_C")))
  (for (_array_element _array_index) _output_get
    (if (Math|Color|NearEqual(LinearColor) _array_element Color)
      (return true _array_index)))
  (return false 0))

(fn GetAttenuationSettings
  (params (CreativeAttenuationRange :type ECreativeAttenuationRange%20Enum :default "NewEnumerator0"))
  (return (Utilities|Map|Find (LocalVariables|GetRange) CreativeAttenuationRange)))

(fn GetColorAtTableIndex
  (params (Table :type Data%20Table%20Object%20Reference) (Index :type Integer))
  (Utilities|IsValid Table
    (:Is%20Valid
      (bind _outrownames (DataTable|GetDataTableRowNames Table))
      (if (Utilities|Array|IsValidIndex _outrownames Index)
        (bind _returnvalue (Utilities|GetDataTableRow Table (Utilities|Array|Get(acopy) _outrownames Index))
          (:then
            (return (Utilities|Struct|BreakUserOptionDefinitionColorValueEntry _returnvalue)))
          (:RowNotFound
            _))
        (else
          _))
      (return "(R=0.998264,G=0.998264,B=0.998264,A=1.000000)"))))

(fn LegacyIntToEnabledPhaseEnum
  (params (LegacyInteger :type Integer))
  (bind _returnvalue (+ LegacyInteger 1))
  ;[1 1 1 1 15] Enum values are shifted by 1 from UserOption values (which start at -1)
  (if (>= _returnvalue 0)
    (return (Utilities|Enum|BytetoEnumEEnabledOnPhase (Math|Conversions|ToByte(Integer) _returnvalue)))
    (else
      ;[1 0.1655 0.1456 1 14] This is never expected to be hit since the saved EnabledIndex property will never have been set to less than -1
      (return "NewEnumerator1"))))
