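; Blueprint interface for Creative devices, declaring common functions through which devices can be queried and activated.
; It covers activation from another device with an instigator and requester type, pawn interaction checks, and toggling trigger collision.
; Queries expose enabled state, device color, instigating controller, spawned pickups, and whether an item is held for a given pawn.
; The functions contain no logic, only signatures with empty default returns.
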(fn ActivateFromOtherDevice
  (params (Instigator :type Controller%20Object%20Reference) (Requester :type ERequesterType%20Enum :default "NewEnumerator0"))
  (return false))

(fn CanInteract
  (params (InteractingPawn :type Fort%20Pawn%20Object%20Reference) (InteractionBeingAttempted :type EInteractionBeingAttempted%20Enum :default "FirstInteraction") (InteractionType :type TInteractionType%20Enum :default "IT_NoInteraction"))
  (return false))

(fn GetDeviceColor
  (return 0))

(fn GetInstigatingController
  (return 0))

(fn GetSpawnedPickups
  (return 0))

(fn IsEnabled
  (return false))

(fn IsHoldingItemForPawn
  (params (Item :type Item%20Variant%20Handle%20Structure) (Pawn :type Pawn%20Object%20Reference))
  (return false))

(fn SetTriggerCollision
  (params (On :type Boolean))
  (return false))
