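; Base Blueprint for Creative devices, named as an edit-only parent, that provides a shared initialization hook driven by device options.
; The ToyOptionsComponent's property-changed and properties-loaded events both call Initialize Device.
; Initialize Device is an empty custom event here, as is the construction script, so this asset adds no setup behavior beyond that wiring.
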(event Custom|Initialize%20Device)

(event OnAnyPropertyChanged(ToyOptionsComponent)
  (CallFunction|InitializeDevice))

(event OnPropertiesLoaded(ToyOptionsComponent)
  (CallFunction|InitializeDevice))

(fn ConstructionScript)
