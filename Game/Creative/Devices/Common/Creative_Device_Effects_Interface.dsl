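; Blueprint interface for Creative devices, named for exposing a device's effect data to other objects.
; It declares two getters, GetCreativeEffectData with a single output and GetCreativeCustomEffectData with five outputs, one of them defaulting to None.
; Both functions hold only default returns and no logic, leaving the actual data to implementing classes.
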(fn GetCreativeCustomEffectData
  (return 0 0 "None" 0 0))

(fn GetCreativeEffectData
  (return 0))
