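; Base parent Blueprint for Creative device props, providing shared minigame, activation, enabled-state, visibility and health behavior for child devices.
; Option changes and loads trigger InitializeDevice, which updates play mode and in-game visibility.
; Minigame state changes drive phase-based enabling or authority-only activation through overridable hooks.
; It also sets max health via a GameplayEffect, swaps move-tool preview components and migrates legacy integer phase options to enums.
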(event OnAddedtoMinigame(FortMinigameProgress)
  (params (Minigame :type Fort%20Minigame%20Object%20Reference))
  ;[1 0.75 0 0.5] Minigame Events
  (Utilities|IsValid Minigame
    (:Is%20Valid
      (Minigame|AssignOnPlayModeChangedDelegate Minigame (EventDispatchers|CreateEvent :!event Play%20Mode%20Changed))
      (Minigame|AssignStateChangedDelegate Minigame (EventDispatchers|CreateEvent :!event Custom|On%20Mingame%20State%20Changed))
      (CallFunction|OnMingameStateChanged Minigame (Variables|/Script/FortniteGame.FortMinigame|GetCurrentState Minigame)))))

(event Custom|OnDevicePlaced
  (params (bIsPreview :type Boolean))
  ;[1 0.586 0 0.2] Custom sound when device is placed
  (bind _deviceplacedsound (Variables|Audio|GetDevicePlacedSound)
    (:then
      (if (not (or (CallFunction|IsinPlayMode) bIsPreview))
        (Audio|PlaySoundatLocation _deviceplacedsound (Transformation|GetActorLocation))))
    (:else)))

(event OnRemovedfromMinigame(FortMinigameProgress)
  (params (Minigame :type Fort%20Minigame%20Object%20Reference))
  ;[1 0.75 0 0.5] Minigame Events
  (Utilities|IsValid Minigame
    (:Is%20Valid
      (Minigame|UnbindEventfromOnPlayModeChangedDelegate Minigame (EventDispatchers|CreateEvent :!event Play%20Mode%20Changed))
      (Minigame|UnbindEventfromStateChangedDelegate Minigame (EventDispatchers|CreateEvent :!event Custom|On%20Mingame%20State%20Changed)))))

(event MinigameStartedDelegate(FortMinigameProgress)
  ;[1 0.75 0 0.5] Minigame Events
  (CallFunction|OnMinigameStarted))

(event Custom|OnMinigameStarted)

(event Custom|OnMinigameEnded)

(event MinigameEndedDelegate(FortMinigameProgress)
  ;[1 0.75 0 0.5] Minigame Events
  (CallFunction|OnMinigameEnded))

(event OnAnyPropertyChanged(ToyOptionsComponent)
  ;[0 0.474 0.05472 0.5] Initialize the device.
  (if (or (and (Variables|Default|GetInitialize%20during%20Preview) (Building|IsPreviewBuildingActor)) (not (Building|IsPreviewBuildingActor)))
    ;[0.9428 0 1 0.2] If we're in editor, fake entering the Pre-Game phase as this is normally called through OnAddedToMinigame, which doesn't happen. Also fake OnWorldReady
    (if (Editor|InEditor)
      (CallFunction|OnMingameStateChanged 0)
      (else
        _))
    (CallFunction|InitializeDevice)))

(event OnPropertiesLoaded(ToyOptionsComponent)
  ;[0 0.474 0.05472 0.5] Initialize the device.
  (if (or (and (Variables|Default|GetInitialize%20during%20Preview) (Building|IsPreviewBuildingActor)) (not (Building|IsPreviewBuildingActor)))
    ;[0.9428 0 1 0.2] If we're in editor, fake entering the Pre-Game phase as this is normally called through OnAddedToMinigame, which doesn't happen. Also fake OnWorldReady
    (if (Editor|InEditor)
      (CallFunction|OnMingameStateChanged 0)
      (else
        _))
    (CallFunction|InitializeDevice)))

(event Custom|On%20Mingame%20State%20Changed
  (params (Minigame :type Fort%20Minigame%20Object%20Reference) (MinigameState :type EFortMinigameState%20Enum :default "PreGame"))
  ;[0 0.432 1 0.2] Whenever the Minigame Game State changes, run Checks on whether we need to change its Play Mode state, Enabled State, and whether we need to activate it.
  (if (Variables|Default|GetUpdate%20on%20Minigame%20State%20Change)
    (Class|CreativeEnabledComponent|SetEnabledOnPhase (Variables|Default|GetEnabledComponent) MinigameState (ChildFunctions|GetEnabledStateAsIndex))
    (MinigamePhaseChecks|ActivateOnPhase MinigameState)))

(event PostMinigameStartedDelegate(FortMinigameProgress)
  ;[0 0.432 1 0.2] Whenever the Minigame Game State changes, run Checks on whether we need to change its Play Mode state, Enabled State, and whether we need to activate it.
  (MinigamePhaseChecks|ActivateOnPhase "InProgress" true))

(event EventOnWorldReady
  ;[1 0 0 0.5] On World Ready - Here is better to do visibility component populating, as there will have been collision updates to StaticMeshComponent after BeginPlay
  (Visibility|UpdateVisibility (Variables|Default|GetVisibleInGameComponent))
  (MoveToolPreview|UpdatePreviewOnlyVisibility))

(event EventBP_SetVisibility
  (params (bVisible :type Boolean))
  (bind _visibleingamecomponent (Variables|Default|GetVisibleInGameComponent))
  (if bVisible
    (Visibility|SetForcedVisibilityDuringGame _visibleingamecomponent "NewEnumerator1")
    (else
      (Visibility|SetForcedVisibilityDuringGame _visibleingamecomponent "NewEnumerator2"))))

(fn ActivateDevice
  ;[1 1 1] Mainly a hook for child blueprints to do something when the Device is activated.
  (Default|CallOnDeviceActivated))

(fn ActivateDeviceFailed)

(fn ActivateonPhase
  (params (State :type EFortMinigameState%20Enum :default "PreGame") (Allow%20Game%20Progress :type Boolean))
  (bind _enabledindex (ChildFunctions|GetActivatedStateAsIndex))
  ;[0 0.432 1] Checks if we want to activate the device, and we're in play mode, and if so activate the device. Utilises the common ActivateOnGamePhase knob.
  (Networking|SwitchHasAuthority
    (:Authority
      (if (or Allow%20Game%20Progress (!= _enabledindex 3))
        (if (CallFunction|IsinPlayMode)
          (if (Class|CreativeCommonDeviceFunctionLibrary|IsActivatedonPhase State _enabledindex self)
            (Activation|AttemptActivateDevice 0)))))))

(fn AttemptActivateDevice
  (params (Instigating%20Controller :type Controller%20Object%20Reference) (Additional%20Requirements :type Boolean :default true))
  ;[1 1 1] Checks to see if the device can be activated. Likely overridden by children.
  (if (and Additional%20Requirements (Class|CreativeEnabledComponent|IsEnabled (Variables|Default|GetEnabledComponent)))
    (Variables|Default|SetInstigating%20Controller Instigating%20Controller)
    (Activation|ActivateDevice)
    (else
      (Activation|ActivateDeviceFailed))))

(fn BlueprintCanInteract
  (params (InteractingPawn :type Fort%20Pawn%20Object%20Reference) (InteractionBeingAttempted :type EInteractionBeingAttempted%20Enum :default "FirstInteraction") (InteractionType :type TInteractionType%20Enum :default "IT_NoInteraction"))
  ;[1 1 1] Allow Interact if we're in Edit mode and have the appropriate components / state.
  ; Handled through a native helper to improve performance.
  (bind _returnvalue (CallFunction|BPCanInteractNativeHelper InteractingPawn (Variables|Default|GetToyOptionsComponent) (Variables|Default|GetFortMinigameProgress)))
  (return _returnvalue))

(fn BlueprintModifyIncomingDamage
  (params (Damage :type Float%20%28single-precision%29) (InTags :type Gameplay%20Tag%20Container%20Structure%20%28by%20ref%29) (EffectContext :type Gameplay%20Effect%20Context%20Handle%20Structure%20%28by%20ref%29) (EventInstigator :type Controller%20Object%20Reference) (DamageCauser :type Actor%20Object%20Reference))
  (if (and (CallFunction|IsinPlayMode) (not (Variables|BuildingActor|GetbIsInvulnerable)))
    (return Damage)
    (else
      (return 0.0))))

(fn BlueprintOnLocalInteract
  (params (InteractingPawn :type Fort%20Player%20Pawn%20Object%20Reference) (InteractionBeingAttempted :type EInteractionBeingAttempted%20Enum :default "FirstInteraction"))
  ;[1 1 1] Never want to Local Interact.
  (return false))

(fn EnabledPhaseEnumToLegacyInt
  (params (EnabledPhase :type Byte :default "NewEnumerator0"))
  ;[1 1 1] Enum values are shifted by 1 from UserOption values (which start at -1)
  (return (- (Math|Conversions|ToInteger(Byte) EnabledPhase) 1)))

(fn GetActivatedStateasIndex
  (return (Variables|LegacyOptions|GetActivateOnGamePhase)))

(fn GetEnabledStateasIndex
  (return (Variables|LegacyOptions|GetEnabledIndex)))

(fn GetMaxHealthOptionValue
  (return (Variables|LegacyOptions|GetMaxHealth)))

(fn GetMoveToolPreviewComponents
  ;[0 0.8698 0.7627 0.25] Defines two lists of objects to hide or show during preview. Use if you want the preview mesh to be different from the normal mesh. Empty (Overridden by Children)
  (return (Utilities|Array|MakeArray) (Utilities|Array|MakeArray)))

(fn GetVisibilityComponents
  (return (Utilities|Array|MakeArray (Variables|Debug|GetStaticMeshComponent))))

(fn GetVisibilityComponentsWithCollisionOverrides
  (return (Utilities|Map|MakeMap)))

(fn InitializeDevice
  ;[1 1 1] Children will very likely extend this.
  (ChildFunctions|PlayModeChanged 0 (CallFunction|IsinPlayMode))
  (CallFunction|UpdateVisibilityInGame)
  (Default|CallOnDeviceInitialized))

(fn IsVisibleInGame
  (return (Variables|LegacyOptions|GetbIsVisible)))

(fn IsInPlayMode
  (return (Minigame|IsinPlayMode self :!class /Script/FortniteGame.MinigameBlueprintLibrary)))

(fn LegacyIntToEnabledPhaseEnum
  (params (LegacyInteger :type Integer))
  (bind _returnvalue (+ LegacyInteger 1))
  ;[1 1 1 1 15] Enum values are shifted by 1 from UserOption values (which start at -1)
  (if (>= _returnvalue 0)
    (return (Utilities|Enum|BytetoEnumEEnabledOnPhase (Math|Conversions|ToByte(Integer) _returnvalue)))
    (else
      ;[1 0.1655 0.1456 1 14] This is never expected to be hit since the saved EnabledIndex property will never have been set to less than -1
      (return "NewEnumerator1"))))

(fn MigrateLegacyActivateOnGamePhase
  (params (ActivateOnPhaseValue :type EActivateOnPhase%20Enum%20%28by%20ref%29 :default "NewEnumerator0") (ClassDefault :type EActivateOnPhase%20Enum :default "NewEnumerator0"))
  (bind _returnvalue (Math|Conversions|ToInteger(Byte) ActivateOnPhaseValue))
  (if (or (Variables|Internal|GetbMigratedActivateOnPhase) (Utilities|Enum|NotEqual(Enum) ActivateOnPhaseValue ClassDefault))
    ;[0.522 0.9099 1 1 13] Backport new enum to ActivateOnGamePhase (which is being used internally)
    (if (!= (Variables|LegacyOptions|GetActivateOnGamePhase) _returnvalue)
      (Variables|LegacyOptions|SetActivateOnGamePhase _returnvalue)
      (Variables|Internal|SetbMigratedActivateOnPhase true)
      (return true))
    (elif (!= (Variables|LegacyOptions|GetActivateOnGamePhase) (Class|GetClassDefaults (Utilities|GetClass self)))
      ;[0.7703 1 0.7232 1 12] Regular Property migration. Leave ActivateOnGamePhase unchanged since it will be used internally
      (Variables|SetBy-RefVar ActivateOnPhaseValue (Utilities|Enum|BytetoEnumEActivateOnPhase (Math|Conversions|ToByte(Integer) (Variables|LegacyOptions|GetActivateOnGamePhase))))
      (Variables|Internal|SetbMigratedActivateOnPhase true)
      (return true)
      (else
        (return false)))))

(fn MigrateLegacyEnabledIndex
  (params (EnabledOnPhaseValue :type EEnabledOnPhase%20Enum%20%28by%20ref%29 :default "NewEnumerator1") (ClassDefault :type EEnabledOnPhase%20Enum :default "NewEnumerator1"))
  (bind _legacyinteger (DeviceConversion|EnabledPhaseEnumtoLegacyInt EnabledOnPhaseValue))
  (if (or (Variables|Internal|GetbMigratedEnabledIndex) (Utilities|Enum|NotEqual(Enum) EnabledOnPhaseValue ClassDefault))
    ;[0.522 0.9099 1 1 13] Backport new enum to EnabledIndex (which is being used internally)
    (if (!= (Variables|LegacyOptions|GetEnabledIndex) _legacyinteger)
      (Variables|LegacyOptions|SetEnabledIndex _legacyinteger)
      (Variables|Internal|SetbMigratedEnabledIndex true)
      (return true))
    (elif (!= (Variables|LegacyOptions|GetEnabledIndex) (Class|GetClassDefaults (Utilities|GetClass self)))
      ;[0.7703 1 0.7232 1 12] Regular Property migration. Leave EnabledIndex unchanged since it will be used internally
      (Variables|SetBy-RefVar EnabledOnPhaseValue (DeviceConversion|LegacyInttoEnabledPhaseEnum (Variables|LegacyOptions|GetEnabledIndex)))
      (Variables|Internal|SetbMigratedEnabledIndex true)
      (return true)
      (else
        (return false)))))

(fn PlayModeChanged
  (params (Minigame :type Fort%20Minigame%20Object%20Reference) (bIsInPlayMode :type Boolean))
  ;[1 1 1] Updates Enabled State when Play Mode Changes. Children will likely extend this.
  ;
  (Default|CallOnPlayModeChanged (Variables|GetbIsInPlayMode))
  (CallFunction|ResetHealthifSet)
  (Networking|SwitchHasAuthority
    (:Authority
      (if (Variables|Default|GetUpdate%20on%20Play%20Mode%20Changed)
        (Class|CreativeEnabledComponent|SetEnabledOnPlayMode (Variables|Default|GetEnabledComponent) (Variables|GetbIsInPlayMode)) ; When play mode changes, enable or disable using the device based on\nwhether it should work in edit mode or not.
      ))))

(fn ResetHealthIfSet
  (bind _self self)
  (bind _returnvalue (Ability|GetAbilitySystemComponent _self))
  (bind _self_2 self)
  (bind _returnvalue_2 (Ability|GetAbilitySystemComponent _self_2))
  (bind _maxhealthoptionvalue (LocalVariables|GetMaxHealthOptionValue))
  (bind _returnvalue_3 (> _maxhealthoptionvalue 0.0))
  (LocalVariables|SetMaxHealthOptionValue (ChildFunctions|GetMaxHealthOptionValue))
  (if (!= (LocalVariables|GetMaxHealthOptionValue) 0.0)
    (CallFunction|SetSuppressHealthBar (not _returnvalue_3))
    (Fortnite|SetCanbeDamaged self _returnvalue_3)
    (LocalVariables|SetHealth (select _returnvalue_3 (LocalVariables|GetMaxHealthOptionValue) 1e+03))
    ;[1 1 1] Applying GE to modify maximum health on this object.
    (Networking|SwitchHasAuthority
      (:Authority
        (Utilities|IsValid _returnvalue_2
          (:Is%20Valid
            (bind _returnvalue_4 (Ability|GameplayEffect|AssignTagSetbyCallerMagnitude (GameplayEffects|MakeOutgoingSpec :self _returnvalue_2 :GameplayEffectClass "/Game/Creative/Devices/Common/GameplayEffects/GE_Devices_Common_SetMaxHealth.GE_Devices_Common_SetMaxHealth_C" :Level 0.0) "(TagName=\"SetByCaller.MaxHealth\")" (LocalVariables|GetHealth)))
            (Utilities|IsValid _returnvalue
              (:Is%20Valid
                (GameplayEffects|RemoveActiveGameplayEffect _returnvalue (Variables|Default|GetMaxHealthEffectHandler) -1)
                (bind _returnvalue_5 (GameplayEffects|ApplyGameplayEffectSpecToSelf _returnvalue _returnvalue_4))
                (Variables|Default|SetMaxHealthEffectHandler _returnvalue_5)
                (Health|SetHealth (Health|GetMaxHealth))))))))))

(fn UpdatePreviewOnlyVisibility
  (bind (_forcehideinpreview _forceshowinpreview) (MoveToolPreview|GetMoveToolPreviewComponents))
  ;[1 0.8055 0 0.25] If we're the preview mesh version of this prop, show / hide all objects marked as such for Move Tool Preview Only
  (Networking|CanExecuteCosmeticEvents
    (:True
      (if (Building|IsPreviewBuildingActor)
        (bind _output_get (LocalVariables|SetComponentsToShow _forceshowinpreview))
        (bind _output_get_2 (LocalVariables|SetComponentsToHide _forcehideinpreview))
        (for _array_element _output_get_2
          (Rendering|SetVisibility _array_element false false)
          (Development|SetHiddeninGame _array_element true false))
        (for _array_element_2 _output_get
          (Rendering|SetVisibility _array_element_2 true false)
          (Development|SetHiddeninGame _array_element_2 false false))))))

(fn UpdateVisibilityinGame
  (bind _visibilitycomponents (CallFunction|GetVisibilityComponents))
  (bind _visibilitycomponents_2 (CallFunction|GetVisibilityComponentswithCollisionOverrides))
  (Initialization|InitializeWithSceneComponents (Variables|Default|GetVisibleInGameComponent) _visibilitycomponents _visibilitycomponents_2 (select (ChildFunctions|IsVisibleinGame) "NewEnumerator3" "NewEnumerator1")))

(fn ConstructionScript
  (|CallParentFunction)
  (if (and (Variables|Default|GetInitialize%20for%20Editor) (Editor|InEditor))
    (CallFunction|InitializeDevice)))
