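; Base Creative prop spawner device that keeps one spawned actor at its spawn location, with stub getters and spawn event for children to override.
; Enable, Spawn, initialization and play-mode changes call authority-only Reset, which replaces the actor in spatial gameplay experiences, throttled during gameplay.
; Actor death schedules a timed Reset: edit-mode delay, or optional gameplay respawn cancelled by Disable.
; DestroyAll, EndPlay and optionally Disable destroy the actor.
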(event EventBeginPlay
  (Creative|Networking|Authority/Cosmetic
    (:Cosmetic
      (if (not (Building|IsPreviewBuildingActor))
        (if (Fortnite|IsSpatialGameplayExperience)
          (Preview|HideEditOnlyMeshes))))))

(event EventInitializeDevice
  ;[1 1 1 1 10] Seems like the class default "Initialize for Editor" setting is not working as expected, so this is here in the meantime
  (if (not (Editor|InEditor))
    (Default|Reset)))

(event Custom|Enable
  (params (Instigator :type Controller%20Object%20Reference))
  ;[1 0.8161 0.5012] Enable/Disable Receivers
  (if (not (Class|CreativeEnabledComponent|IsEnabled (Variables|Default|GetEnabledComponent)))
    (Class|CreativeEnabledComponent|SetEnabled (Variables|Default|GetEnabledComponent) true)
    (if (UserOptions|VirtualGetters|GetSpawnwhenEnabled)
      (Default|Reset))))

(event EventEndPlay
  (params (EndPlayReason :type EEndPlayReason%20Enum :default "Destroyed"))
  (CallFunction|DestroyExistingActors))

(event Custom|Disable
  (params (Instigator :type Controller%20Object%20Reference))
  ;[1 0.8161 0.5012] Enable/Disable Receivers
  (if (Class|CreativeEnabledComponent|IsEnabled (Variables|Default|GetEnabledComponent))
    (Class|CreativeEnabledComponent|SetEnabled (Variables|Default|GetEnabledComponent) false)
    (Utilities|Time|ClearandInvalidateTimerbyHandle (Variables|Default|GetGameplayRespawnTimerHandle))
    (if (Variables|UserOptions|GetDestroy%20When%20Disabled)
      (CallFunction|DestroyExistingActors))))

(event OnAddedtoMinigame(FortMinigameProgress)
  (params (Minigame :type Fort%20Minigame%20Object%20Reference))
  (Networking|SwitchHasAuthority
    (:Authority
      (Minigame|AssignOnPlayModeChangedDelegate Minigame (delegate Custom|OnMinigamePlayModeChanged)))))

(event Custom|Spawn
  (params (Instigator :type Controller%20Object%20Reference))
  ;[1 0.25 0.25] Other Receiver Events
  (if (Class|CreativeEnabledComponent|IsEnabled (Variables|Default|GetEnabledComponent))
    (Default|Reset)))

(event Custom|OnMinigamePlayModeChanged
  (params (Minigame :type Fort%20Minigame%20Object%20Reference) (bIsInPlayMode :type Boolean))
  (Default|Reset))

(event Custom|DestroyAll
  (params (Instigator :type Controller%20Object%20Reference))
  ;[1 0.25 0.25] Other Receiver Events
  (if (Class|CreativeEnabledComponent|IsEnabled (Variables|Default|GetEnabledComponent))
    (CallFunction|DestroyExistingActors)))

(fn DestroyExistingActors
  (bind _initialactor (Variables|Default|GetInitialActor)
    (:then
      (Actor|DestroyActor _initialactor)
      (Variables|Default|SetInitialActor 0))
    (:else)))

(fn GetConnectedVisibleActor
  (return (Variables|Default|GetInitialActor)))

(fn GetDoesRespawnOnTimer
  (return false))

(fn GetEditOnlyMeshes
  (return (Utilities|Array|MakeArray (Variables|Default|GetCreativeEditOnlyMesh))))

(fn GetGameplayRespawnDelaySeconds
  (return 0.0))

(fn GetSpawnWhenEnabled
  (return false))

(fn HideEditOnlyMeshes
  (for _array_element (Preview|GetEditOnlyMeshes)
    (Utilities|IsValid _array_element
      (:Is%20Valid
        (Rendering|SetVisibility _array_element false false)))))

(fn OnInitialActorDied
  (params (DamagedActor :type Actor%20Object%20Reference) (Damage :type Float%20%28single-precision%29) (InstigatedBy :type Controller%20Object%20Reference) (DamageCauser :type Actor%20Object%20Reference) (HitLocation :type Vector) (FHitComponent :type Primitive%20Component%20Object%20Reference) (BoneName :type Name) (Momentum :type Vector))
  (Utilities|FlowControl|Sequence
    (:then_0
      (Networking|SwitchHasAuthority
        (:Authority
          (if (Class|CreativeCommonDeviceFunctionLibrary|IsLogicinPlayMode (Variables|Default|GetFortMinigameProgress))
            (CallFunction|OnInitialObjectDestroyedDuringGameplay InstigatedBy)
            (elif (!= (Variables|GetDamageCauser) (Variables|Default|GetInitialActor))
              (Utilities|Time|SetTimerbyEvent (EventDispatchers|CreateEvent :!event Reset) (Variables|Default|GetEditModeDelayRespawnSeconds)) ; Respawn in edit mode
            )))))
    (:then_1
      (Variables|Default|SetInitialActor 0))))

(fn OnInitialObjectDestroyedDuringGameplay
  (params (DestroyedInstigator :type Controller%20Object%20Reference))
  (if (UserOptions|VirtualGetters|GetDoesRespawnonTimer)
    (bind _returnvalue (Utilities|Time|SetTimerbyEvent (EventDispatchers|CreateEvent :!event Reset) (UserOptions|VirtualGetters|GetGameplayRespawnDelaySeconds))) ; Respawn in gameplay. May be cancelled if base is destroyed and required for respawn.
    (Variables|Default|SetGameplayRespawnTimerHandle _returnvalue)))

(fn Reset
  (Networking|SwitchHasAuthority
    (:Authority
      (if (or (not (Class|CreativeCommonDeviceFunctionLibrary|IsLogicinPlayMode (Variables|Default|GetFortMinigameProgress))) (and (Class|CreativeEnabledComponent|IsEnabled (Variables|Default|GetEnabledComponent)) (or (not (Variables|SpawnThrottle|GetbThrottlePreventAction)) (not (Minigame|IsMinigameinProgress (Variables|Default|GetFortMinigameProgress) :!class /Script/FortniteGame.FortMinigameLogicComponent)))))
        (if (Fortnite|IsSpatialGameplayExperience)
          (CallFunction|DestroyExistingActors)
          (CallFunction|SpawnInitialActor)
          (SpawnThrottle|ThrottleActionsDuringGameplay))))))

(fn ResetThrottle
  (Variables|SpawnThrottle|SetbThrottlePreventAction false))

(fn SpawnInitialActor
  (bind _spawnedactor (Fortnite|Utilities|SpawnTransientActor (Variables|Default|GetInitialActorClass) (Transformation|GetWorldTransform (Variables|Default|GetSpawnLocation)) "Undefined" self))
  (bind _output_get (Variables|Default|SetInitialActor _spawnedactor))
  (InteractInterface|SetActorforInteraction _output_get self)
  (UserOptions|VirtualEvents|TransmitObjectSpawned (Utilities|Struct|MakeGameplayEventSpawned 0))
  (EventDispatchers|BindEventtoOnDied (Variables|Default|GetInitialActor) (EventDispatchers|CreateEvent :!event OnInitialActorDied)))

(fn ThrottleActionsDuringGameplay
  (if (Class|CreativeCommonDeviceFunctionLibrary|IsLogicinPlayMode (Variables|Default|GetFortMinigameProgress))
    (Variables|SpawnThrottle|SetbThrottlePreventAction true)
    (Utilities|Time|SetTimerbyEvent (EventDispatchers|CreateEvent :!event ResetThrottle) (Variables|SpawnThrottle|GetThrottleSeconds))))

(fn TransmitObjectSpawned
  (params (Event :type Gameplay%20Event%20Spawned%20Structure)))

(fn ConstructionScript
  (|CallParentFunction))
