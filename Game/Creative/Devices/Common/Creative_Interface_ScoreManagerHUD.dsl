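; Blueprint interface, judging by its name, for the score manager HUD among Creative devices.
; It declares one function, MessageScore, which returns four values that default to an empty text value, two zeros and false.
; The file holds no logic beyond that default return.
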(fn MessageScore
  (return "" 0 0 false))
