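; Base class for Creative level instance devices, providing shared minigame, activation, visibility and health behavior with overridable hooks for children.
; Option loads or changes initialize the device, while bound minigame state and play-mode changes drive its enabled component and server-side phase activation, gated on being enabled.
; Visibility components populate on world ready to follow late collision updates, and move-tool previews can swap components.
; Max health applies through a SetByCaller gameplay effect, with damage only in play mode.
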(event EventBeginPlay
  (CallFunction|UpdateVisibilityinGame)
  (MoveToolPreview|UpdatePreviewOnlyVisibility))

(event OnAddedtoMinigame(FortMinigameProgress)
  (params (Minigame :type Fort%20Minigame%20Object%20Reference))
  ;[1 0.75 0 0.5] Minigame Events
  (Utilities|IsValid Minigame
    (:Is%20Valid
      (Minigame|AssignOnPlayModeChangedDelegate Minigame (EventDispatchers|CreateEvent :!event Play%20Mode%20Changed))
      (Minigame|AssignStateChangedDelegate Minigame (EventDispatchers|CreateEvent :!event Custom|On%20Mingame%20State%20Changed))
      (CallFunction|OnMingameStateChanged Minigame (Variables|/Script/FortniteGame.FortMinigame|GetCurrentState Minigame)))))

(event OnRemovedfromMinigame(FortMinigameProgress)
  (params (Minigame :type Fort%20Minigame%20Object%20Reference))
  ;[1 0.75 0 0.5] Minigame Events
  (Utilities|IsValid Minigame
    (:Is%20Valid
      (Minigame|UnbindEventfromOnPlayModeChangedDelegate Minigame (EventDispatchers|CreateEvent :!event Play%20Mode%20Changed))
      (Minigame|UnbindEventfromStateChangedDelegate Minigame (EventDispatchers|CreateEvent :!event Custom|On%20Mingame%20State%20Changed)))))

(event MinigameStartedDelegate(FortMinigameProgress)
  ;[1 0.75 0 0.5] Minigame Events
  (CallFunction|OnMinigameStarted))

(event Custom|OnMinigameStarted)

(event Custom|OnMinigameEnded)

(event MinigameEndedDelegate(FortMinigameProgress)
  ;[1 0.75 0 0.5] Minigame Events
  (CallFunction|OnMinigameEnded))

(event OnAnyPropertyChanged(ToyOptionsComponent)
  ;[0 0.474 0.05472 0.5] Initialize the device.
  (if (or (and (Variables|Default|GetInitialize%20during%20Preview) (Building|IsPreviewBuildingActor)) (not (Building|IsPreviewBuildingActor)))
    (CallFunction|InitializeDevice)))

(event OnPropertiesLoaded(ToyOptionsComponent)
  ;[0 0.474 0.05472 0.5] Initialize the device.
  (if (or (and (Variables|Default|GetInitialize%20during%20Preview) (Building|IsPreviewBuildingActor)) (not (Building|IsPreviewBuildingActor)))
    (CallFunction|InitializeDevice)))

(event Custom|On%20Mingame%20State%20Changed
  (params (Minigame :type Fort%20Minigame%20Object%20Reference) (MinigameState :type EFortMinigameState%20Enum :default "PreGame"))
  ;[0 0.432 1 0.2] Whenever the Minigame Game State changes, run Checks on whether we need to change its Play Mode state, Enabled State, and whether we need to activate it.
  (if (Variables|Default|GetUpdate%20on%20Minigame%20State%20Change)
    (Class|CreativeEnabledComponent|SetEnabledOnPhase (Variables|Default|GetEnabledComponent) MinigameState (Variables|UserOptions|GetEnabledIndex))
    (MinigamePhaseChecks|ActivateOnPhase MinigameState)))

(event EventOnWorldReady
  ;[1 0 0 0.5] On World Ready - Here is better to do visibility component populating, as there will have been collision updates to StaticMeshComponent after BeginPlay
  (bind _visibilitycomponents (CallFunction|GetVisibilityComponents))
  (Visibility|SetVisibilityComponents (Variables|Default|GetVisibleInGameComponent) _visibilitycomponents true)
  (bind _visibilitycomponents_2 (CallFunction|GetVisibilityComponentswithCollisionOverrides))
  (Visibility|SetVisibilityComponentsWithCollision (Variables|Default|GetVisibleInGameComponent) _visibilitycomponents_2 true))

(fn ActivateDevice
  ;[1 1 1] Mainly a hook for child blueprints to do something when the Device is activated.
  (Default|CallOnDeviceActivated))

(fn ActivateDeviceFailed)

(fn ActivateonPhase
  (params (State :type EFortMinigameState%20Enum :default "PreGame"))
  ;[0 0.432 1] Checks if we want to activate the device, and we're in play mode, and if so activate the device. Utilises the common ActivateOnGamePhase knob.
  (Networking|SwitchHasAuthority
    (:Authority
      (if (CallFunction|IsinPlayMode)
        (if (Class|CreativeCommonDeviceFunctionLibrary|IsActivatedonPhase State (Variables|UserOptions|GetActivateOnGamePhase))
          (Activation|AttemptActivateDevice 0))))))

(fn AttemptActivateDevice
  (params (Instigating%20Controller :type Controller%20Object%20Reference) (Additional%20Requirements :type Boolean :default true))
  ;[1 1 1] Checks to see if the device can be activated. Likely overridden by children.
  (if (and Additional%20Requirements (Class|CreativeEnabledComponent|IsEnabled (Variables|Default|GetEnabledComponent)))
    (Variables|Default|SetInstigating%20Controller Instigating%20Controller)
    (Activation|ActivateDevice)
    (else
      (Activation|ActivateDeviceFailed))))

(fn BlueprintCanInteract
  (params (InteractingPawn :type Fort%20Pawn%20Object%20Reference) (InteractionBeingAttempted :type EInteractionBeingAttempted%20Enum :default "FirstInteraction") (InteractionType :type TInteractionType%20Enum :default "IT_NoInteraction"))
  (bind _returnvalue (Pawn|GetController InteractingPawn))
  ;[1 1 1] Allow Interact if we're in Edit mode and have the appropriate components / state.
  (Utilities|FlowControl|Sequence
    (:then_0
      (bind _toyoptionscomponent (Variables|Default|GetToyOptionsComponent)
        (:then
          (Variables|/Script/FortniteGame.FortActorOptionsComponent|GetInternal_PlayerOptions _toyoptionscomponent
            (:then
              (Utilities|IsValid InteractingPawn
                (:Is%20Valid
                  (Utilities|IsValid _returnvalue
                    (:Is%20Valid
                      (bind _asfort_player_controller_athena (Utilities|Casting|CastToFortPlayerControllerAthena _returnvalue))
                      (if (.ReturnValue (Creative|CanPlayerCreateinCurrentVolume _asfort_player_controller_athena :!class /Script/FortniteGame.FortPlayerControllerAthena))
                        (LocalVariables|SetbInteractSuccess (not (CallFunction|IsinPlayMode)))))))))
            (:else)))
        (:else)))
    (:then_1
      (return (LocalVariables|GetbInteractSuccess)))))

(fn BlueprintModifyIncomingDamage
  (params (Damage :type Float%20%28single-precision%29) (InTags :type Gameplay%20Tag%20Container%20Structure%20%28by%20ref%29) (EffectContext :type Gameplay%20Effect%20Context%20Handle%20Structure%20%28by%20ref%29) (EventInstigator :type Controller%20Object%20Reference) (DamageCauser :type Actor%20Object%20Reference))
  (if (and (CallFunction|IsinPlayMode) (not (Variables|BuildingActor|GetbIsInvulnerable)))
    (return Damage)
    (else
      (return 0.0))))

(fn BlueprintOnLocalInteract
  (params (InteractingPawn :type Fort%20Player%20Pawn%20Object%20Reference) (InteractionBeingAttempted :type EInteractionBeingAttempted%20Enum :default "FirstInteraction"))
  ;[1 1 1] Never want to Local Interact.
  (return false))

(fn GetMoveToolPreviewComponents
  ;[0 0.8698 0.7627 0.25] Defines two lists of objects to hide or show during preview. Use if you want the preview mesh to be different from the normal mesh. Empty (Overridden by Children)
  (return (Utilities|Array|MakeArray 0) (Utilities|Array|MakeArray 0)))

(fn GetVisibilityComponents
  (return (Utilities|Array|MakeArray (Variables|Debug|GetStaticMeshComponent))))

(fn GetVisibilityComponentsWithCollisionOverrides
  (return (Utilities|Map|MakeMap (Variables|Debug|GetStaticMeshComponent) "QueryAndPhysics")))

(fn InitializeDevice
  ;[1 1 1] Children will very likely extend this.
  (ChildFunctions|PlayModeChanged 0 (CallFunction|IsinPlayMode))
  (CallFunction|UpdateVisibilityinGame)
  (Default|CallOnDeviceInitialized))

(fn IsVisibleInGame
  (return (Variables|UserOptions|GetbIsVisible)))

(fn IsInPlayMode
  (bind _currentminigame (Variables|/Script/FortniteGame.FortMinigameLogicComponent|GetCurrentMinigame (Variables|Default|GetFortMinigameProgress))
    (:then
      (return (Minigame|IsinPlayMode _currentminigame :!class /Script/FortniteGame.FortMinigame)))
    (:else
      (return false))))

(fn PlayModeChanged
  (params (Minigame :type Fort%20Minigame%20Object%20Reference) (bIsInPlayMode :type Boolean))
  ;[1 1 1] Updates Enabled State when Play Mode Changes. Children will likely extend this.
  ;
  (Default|CallOnPlayModeChanged (Variables|GetbIsInPlayMode))
  (CallFunction|ResetHealthifSet)
  (Networking|SwitchHasAuthority
    (:Authority
      (if (Variables|Default|GetUpdate%20on%20Play%20Mode%20Changed)
        (Class|CreativeEnabledComponent|SetEnabledOnPlayMode (Variables|Default|GetEnabledComponent) (Variables|GetbIsInPlayMode)) ; When play mode changes, enable or disable using the device based on\nwhether it should work in edit mode or not.
      ))))

(fn ResetHealthIfSet
  (bind _self self)
  (bind _returnvalue (Ability|GetAbilitySystemComponent _self))
  (bind _maxhealth (Variables|UserOptions|GetMaxHealth))
  (bind _returnvalue_2 (> _maxhealth 0.0))
  (if (!= (Variables|UserOptions|GetMaxHealth) 0.0)
    (CallFunction|SetSuppressHealthBar (not _returnvalue_2))
    (Fortnite|SetCanbeDamaged self _returnvalue_2)
    (LocalVariables|SetHealth (select _returnvalue_2 (Variables|UserOptions|GetMaxHealth) 1e+03))
    ;[1 1 1] Applying GE to modify maximum health on this object.
    (Networking|SwitchHasAuthority
      (:Authority
        (bind _returnvalue_3 (Ability|GameplayEffect|AssignTagSetbyCallerMagnitude (GameplayEffects|MakeOutgoingSpec :self (Ability|GetAbilitySystemComponent self) :GameplayEffectClass "/Game/Creative/Devices/Common/GameplayEffects/GE_Devices_Common_SetMaxHealth.GE_Devices_Common_SetMaxHealth_C" :Level 0.0) "(TagName=\"SetByCaller.MaxHealth\")" (LocalVariables|GetHealth)))
        (Utilities|IsValid _returnvalue
          (:Is%20Valid
            (GameplayEffects|RemoveActiveGameplayEffect _returnvalue (Variables|Default|GetMaxHealthEffectHandler) -1)
            (bind _returnvalue_4 (GameplayEffects|ApplyGameplayEffectSpecToSelf _returnvalue _returnvalue_3))
            (Variables|Default|SetMaxHealthEffectHandler _returnvalue_4)
            (Health|SetHealth (Health|GetMaxHealth))))))))

(fn UpdatePreviewOnlyVisibility
  (bind (_forcehideinpreview _forceshowinpreview) (MoveToolPreview|GetMoveToolPreviewComponents))
  ;[1 0.8055 0 0.25] If we're the preview mesh version of this prop, show / hide all objects marked as such for Move Tool Preview Only
  (Networking|CanExecuteCosmeticEvents
    (:True
      (if (Building|IsPreviewBuildingActor)
        (for _array_element _forcehideinpreview
          (Rendering|SetVisibility _array_element false false)
          (Development|SetHiddeninGame _array_element true false))
        (for _array_element_2 _forceshowinpreview
          (Rendering|SetVisibility _array_element_2 true false)
          (Development|SetHiddeninGame _array_element_2 false false))))))

(fn UpdateVisibilityInGame
  (Visibility|SetVisiblityDuringPhase (Variables|Default|GetVisibleInGameComponent) (select (ChildFunctions|IsVisibleinGame) "NewEnumerator3" "NewEnumerator1")))

(fn ConstructionScript
  (|CallParentFunction)
  (if (and (Variables|Default|GetInitialize%20for%20Editor) (Editor|InEditor))
    (CallFunction|InitializeDevice)))
