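; Blueprint interface for Creative devices that exposes a mannequin and lets callers play emotes on it.
; It declares getters for the mannequin, its body and gender, and its current montage, plus PlayExternalEmote, which takes a soft reference to an anim montage.
; The functions carry no logic and return only empty or invalid defaults, leaving behavior to implementers.
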(fn GetMannequin
  (return 0))

(fn GetMannequinBodyAndGender
  (return "NONE" "Invalid"))

(fn GetMannequinCurrentMontage
  (return 0))

(fn PlayExternalEmote
  (params (EmoteMontage :type Anim%20Montage%20Soft%20Object%20Reference)))
