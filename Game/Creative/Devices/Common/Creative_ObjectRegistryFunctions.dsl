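; Blueprint function library for Creative devices that tracks objects in shared registers identified by gameplay tag and integer index.
; GetOrCreateObjectRegister finds the tagged CreativeIndexedObjectRegisterComp on the minigame's volume or adds one, optionally registering the context actor.
; CreateRegisterAndUpdateActorIndex moves an actor between indices, updating the caller's cached register and index by reference.
; GetObjectInRegister and GetObjectsInRegister resolve the register first, so callers can cache it to avoid repeated lookups.
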(fn CreateRegisterAndUpdateActorIndex
  (params (Register%20Reference%20%28Creates%20if%20Invalid%29 :type Creative%20Indexed%20Object%20Register%20Comp%20Object%20Reference%20%28by%20ref%29) (Register%20Tag :type Gameplay%20Tag%20Structure) (Actor%20to%20Register :type Actor%20Object%20Reference) (Remove%20from%20Index%20%28Updates%20to%20Index%20after%20Registration%29 :type Integer%20%28by%20ref%29) (Register%20at%20Index :type Integer))
  (bind _register_reference__creates_if_invalid (Variables|GetRegister%20Reference%20%28Creates%20if%20Invalid%29))
  ;[0 0.7912 1 0.5] For a passed in variables, Create a register if it doesn't exist already with the appropriate Tag, Remove actor from Cached Previous Index, Add to the New Index. Will update Cached Previous Index and Register Reference.
  (Utilities|IsValid Register%20Reference%20%28Creates%20if%20Invalid%29
    (:Is%20Valid
      ;;[0 1 0.1052 0.5] Remove the Mannequin from where it was previously registered (if any) and add it to our new register.
      (if (!= (Variables|GetRegister%20at%20Index) (Variables|GetRemove%20from%20Index%20%28Updates%20to%20Index%20after%20Registration%29))
        (Class|CreativeIndexedObjectRegisterComp|RemoveObjectfromIndex _register_reference__creates_if_invalid (Variables|GetActor%20to%20Register) (Variables|GetRemove%20from%20Index%20%28Updates%20to%20Index%20after%20Registration%29))
        (if (>= (Variables|GetRegister%20at%20Index) 0)
          (Class|CreativeIndexedObjectRegisterComp|RegisterObject _register_reference__creates_if_invalid (Variables|GetActor%20to%20Register) (Variables|GetRegister%20at%20Index)))))
    (:Is%20Not%20Valid
      ;;[0 1 0.1052 0.5] This object has not cached a reference to the register yet. Do so now and register with it at the appropriate customization slot.
      (if (>= Register%20at%20Index 0)
        (CreativeObjectRegistry|GetOrCreateObjectRegister (Variables|GetActor%20to%20Register) (Variables|GetRegister%20Tag) true (Variables|GetRegister%20at%20Index) (Variables|GetRegister%20Reference%20%28Creates%20if%20Invalid%29) (Variables|GetActor%20to%20Register)))))
  (Variables|SetBy-RefVar (Variables|GetRemove%20from%20Index%20%28Updates%20to%20Index%20after%20Registration%29) (Variables|GetRegister%20at%20Index)))

(fn GetObjectInRegister
  (params (World%20Context%20Actor :type Actor%20Object%20Reference) (Register%20Tag :type Gameplay%20Tag%20Structure) (Element :type EFortArrayElementPosition%20Enum :default "First") (Register%20Index :type Integer))
  (bind _local_register (LocalVariables|Getlocal%20Register))
  ;[0 0.7912 1 0.5] Gets an object in a Creative Object Register at a specific tag and index. If using this repeatedly save a copy to the register and call 'Get Registered Object directly' (to avoid overheads of finding the register).
  (CreativeObjectRegistry|GetOrCreateObjectRegister World%20Context%20Actor Register%20Tag false Register%20Index (LocalVariables|Getlocal%20Register) World%20Context%20Actor)
  (bind _returnvalue (Class|FortCreativeIndexedObjectRegisterComponent|GetRegisteredObject _local_register Element Register%20Index))
  (return _local_register (Utilities|IsValid _returnvalue :!class /Script/Engine.KismetSystemLibrary) _returnvalue))

(fn GetObjectsInRegister
  (params (World%20Context%20Actor :type Actor%20Object%20Reference) (Register%20Tag :type Gameplay%20Tag%20Structure) (Register%20Index :type Integer))
  (bind _local_register (LocalVariables|Getlocal%20Register))
  ;[0 0.7912 1 0.5] Gets an object in a Creative Object Register at a specific tag and index. If using this repeatedly save a copy to the register and call 'Get Registered Object directly' (to avoid overheads of finding the register).
  (CreativeObjectRegistry|GetOrCreateObjectRegister World%20Context%20Actor Register%20Tag false Register%20Index (LocalVariables|Getlocal%20Register) World%20Context%20Actor)
  (bind (_bisvalid _register) (Class|FortCreativeIndexedObjectRegisterComponent|GetRegisteratIndex _local_register Register%20Index))
  (bind _register_objects (Utilities|Struct|BreakFortObjectArray _register))
  (return _local_register _bisvalid _register_objects))

(fn GetorCreateObjectRegister
  (params (World%20Context%20Actor :type Actor%20Object%20Reference) (Register%20Tag :type Gameplay%20Tag%20Structure) (Register%20Context%20Actor :type Boolean) (Register%20Index :type Integer) (Register%20Ref :type Creative%20Indexed%20Object%20Register%20Comp%20Object%20Reference%20%28by%20ref%29))
  (bind _world_context_actor (Variables|GetWorld%20Context%20Actor))
  (bind _returnvalue (Minigame|GetMinigame _world_context_actor :!class /Script/FortniteGame.MinigameBlueprintLibrary))
  ;[0 0.7912 1 0.5] Returns an Actor Registry with a specific tag if one exists, or creates one if it does not. Actor Registries are simply arrays of actors stored in a central, known location that can be added to or removed from.
  ; This is useful to use when you want to keep track of a group of items and any changes to them.
  (Utilities|FlowControl|Sequence
    (:then_0
      (bind _register_ref (Variables|GetRegister%20Ref)
        (:then
          (if (GameplayTags|NotEqual(GameplayTag) (Variables|GetRegister%20Tag) (Class|CreativeIndexedObjectRegisterComp|GetRegisterTag _register_ref))
            _))
        (:else
          _))
      (Utilities|IsValid _returnvalue
        (:Is%20Valid
          (LocalVariables|SetVolume (Variables|/Script/FortniteGame.FortMinigame|GetVolume _returnvalue))
          (bind _array_element (Utilities|Array|ForEachLoopwithBreak (Actor|GetComponentsByClass (LocalVariables|GetVolume) "/Game/Creative/Devices/Common/Components/CreativeIndexedObjectRegisterComp.CreativeIndexedObjectRegisterComp_C"))
            (:LoopBody
              (if (GameplayTags|Equal(GameplayTag) (Class|CreativeIndexedObjectRegisterComp|GetRegisterTag _array_element) (Variables|GetRegister%20Tag))
                (Variables|SetBy-RefVar (Variables|GetRegister%20Ref) _array_element)
                (LocalVariables|SetfoundRegister true)))
            (:Completed
              _))
          (if (not (LocalVariables|GetfoundRegister))
            (bind _returnvalue_2 (Game|AddComponentbyClass (LocalVariables|GetVolume) 0 "/Game/Creative/Devices/Common/Components/CreativeIndexedObjectRegisterComp.CreativeIndexedObjectRegisterComp_C" (Variables|GetRegister%20Tag)))
            (Variables|SetBy-RefVar (Variables|GetRegister%20Ref) _returnvalue_2)))))
    (:then_1
      ;;[1 1 1] Once Register has been created, register the actor if requested.
      (if (Variables|GetRegister%20Context%20Actor)
        (bind _register_ref_2 (Variables|GetRegister%20Ref)
          (:then
            (Class|CreativeIndexedObjectRegisterComp|RegisterObject _register_ref_2 (Variables|GetWorld%20Context%20Actor) (Variables|GetRegister%20Index)))
          (:else))))))
