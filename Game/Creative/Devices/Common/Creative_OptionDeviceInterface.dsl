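; Blueprint interface for Creative devices that expose options, judging by its name and location, with a query for whether a device has UEFN-only options.
; It declares a single function, HasUEFNOnlyOptions, which returns a boolean and a second text-like value.
; The function as shown returns false and an empty value, with no further logic.
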(fn HasUEFNOnlyOptions
  (return false ""))
